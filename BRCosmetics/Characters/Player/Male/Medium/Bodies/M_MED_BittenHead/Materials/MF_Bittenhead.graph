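# Material function for the BittenHead character cosmetic, replacing standard shading with stylized emissive lighting and flickering, old-film-like screen effects.
# Input base color is relit using sun direction, normals, height-lerped shading, highlights and rim light, tuned by time of day, then output as emissive over black base color.
# A camera-facing, FOV-scaled virtual plane at the actor drives time-stepped strobe, noise and scanlines.
# IsSunglasses switches to a dark lens with view-normal reflection and Fresnel opacity.
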MaterialFunction /BRCosmetics/Characters/Player/Male/Medium/Bodies/M_MED_BittenHead/Materials/MF_Bittenhead
graph {nodes [
  Constant3Vector() {desc=Locker, constant=<rgba 0.6 0.648 0.47 1>}
  Constant3Vector() {desc=Lobby, constant=<rgba -0.648 0.6 0.47 1>}
  Constant() {r=4}
  Constant() {r=0.8}
  Constant() {r=1}
  Constant() {r=0.4}
  Constant() {r=0.4}
  Constant() {r=0.4}
  Constant() {r=4}
  Constant() {r=3}
  Constant() {r=3}
  Constant() {r=2}
  Constant() {r=0.03}
  Constant() {r=1.25}
  Constant() {r=3}
  Constant() {r=-3}
  Constant() {r=0.01}
  Constant() {r=1.2}
  Constant() {r=1.75}
  Constant() {r=2.25}
  n1 = VectorParameter() {parameterName=BHG-Color, defaultValue=<rgba 0.005 0.0066 0.01 1>}
  n2 = Constant() {desc=Needs at least a small amount of this or it will break, r=0.25}
  n3 = MaterialFunctionCall() {materialFunction=&CameraDirectionVector}
  n4 = Normalize(VectorInput=Multiply(A=Constant3Vector() {constant=<rgba 1 1 0 0>}, B=@n3.Result))
  n5 = Constant() {r=25}
  n6 = Add(A=ActorPositionWS(), B=Constant3Vector() {constant=<rgba 0 0 40 1>})
  n7 = MaterialFunctionCall(
   Input=MaterialFunctionCall() {materialFunction=/Engine/Functions/Engine_MaterialFunctions02/Utility/FOV}."FOV "
   Input Low=Constant() {r=0.414}
   Input High=Constant() {r=1.091}
   Target Low=Constant()
   Target High=Constant() {r=1}
  ) {materialFunction=&RemapValueRange}
  n8 = MaterialFunctionCall(
   Plane Normal Axis=Add(A=Multiply(A=@n2, B=@n4), B=@n3.Result)
   Plane Center=Add(A=Multiply(A=@n3.Result, B=@n5), B=@n6)
   UV Scale=LinearInterpolate(
    A=ScalarParameter() {defaultValue=0.8, group=Bittenhead, parameterName=BH_Scale-LowFOV}
    B=ScalarParameter() {defaultValue=1.25, group=Bittenhead, parameterName=BH_Scale-HighFOV}
    Alpha=Saturate(None=@n7.Result)
   )
  ) {materialFunction=&VirtualPlaneCoordinates}
  n9 = Multiply(A=ScalarParameter() {defaultValue=0.01, group=Bittenhead, parameterName=BH_Pulse-Scale}, B=@n8.UVs)
  n10 = ScalarParameter() {defaultValue=16, group=Bittenhead, parameterName=BH_Noise-NumberOffsets}
  n11 = Frac(None=Time())
  n12 = Multiply(A=ScalarParameter() {defaultValue=0.9, group=Bittenhead, parameterName=BH_Noise-OffsetUV}, B=Floor(None=Multiply(A=@n10, B=@n11)))
  n13 = LinearInterpolate(
   A=ScalarParameter() {defaultValue=0.9, group=Bittenhead, parameterName=BH_StrobeA}
   B=ScalarParameter() {defaultValue=1.2, group=Bittenhead, parameterName=BH_StrobeB}
   Alpha=TextureSample(UVs=Add(A=@n9, B=@n12)) {samplerSource=SSM_Wrap_WorldGroupSettings, texture=/Game/Global/Textures/Default/Masks_and_Gradients/T_Vert_Gradient_Sections}.R
  )
  n14 = ScalarParameter() {defaultValue=0.015, group=Bittenhead, parameterName=BH_Noise-Scale}
  n15 = TextureSample(UVs=Add(A=@n12, B=Multiply(A=@n14, B=@n8.UVs))) {
   samplerSource=SSM_Wrap_WorldGroupSettings
   samplerType=SAMPLERTYPE_LinearColor
   texture=/Game/Global/Textures/Default/Noise/SparseGaussianNoise256
  }
  n16 = Multiply(A=OneMinus(None=@n15.R), B=ScalarParameter() {defaultValue=0.03, group=Bittenhead, parameterName=BH_NoiseAmt})
  n17 = Multiply(
   A=ScalarParameter() {defaultValue=0.01, group=Bittenhead, parameterName=BH_ScanlineAmt}
   B=TextureSample(UVs=@n8.UVs) {samplerSource=SSM_Wrap_WorldGroupSettings, texture=/Game/Global/Textures/Default/Masks_and_Gradients/gradient_stasis}.R
  )
  n18 = ScalarParameter() {defaultValue=0.5, group=Bittenhead, parameterName=BH_FX-Amount-Other}
  n19 = Multiply(
   A=Multiply(A=Add(A=@n16, B=@n17), B=VectorParameter() {group=Bittenhead, parameterName=BH_Fx-Color, defaultValue=<rgba 1 1 1 1>}.RGB)
   B=MaterialFunctionCall(Value_Morning=@n18, Value_Day=Constant() {r=1}, Value_Evening=@n18, Value_Night=@n18) {materialFunction=&Scalar_Time_of_Day}.Result
  )
  n20 = Add(A=@n13, B=@n19)
  n21 = Constant3Vector() {constant=<rgba 0 0 1 0>}
  n22 = Constant3Vector() {constant=<rgba 0.5 0.5 0 1>}
  n23 = Add(A=Multiply(A=Transform(None=@n21) {transformType=TRANSFORM_View}, B=@n22), B=@n22)
  n24 = Desaturation(
   None=TextureSample(UVs=ComponentMask(None=@n23) {g=true, r=true}) {texture=/Game/VisualThreatManager/StormVisuals/LargeClouds/T_EyeballReflection2}.RGB
   Fraction=Constant() {r=1}
  )
  n25 = Constant3Vector() {constant=<rgba 0.04 0.0596 0.1 1>}
  n26 = Multiply(A=TwoSidedSign(), B=@n21)
  n27 = LinearInterpolate(A=Constant() {r=0.8}, B=Constant() {r=1}, Alpha=Fresnel(ExponentIn=Constant() {r=1}, Normal=Transform(None=@n26)))
  n28 = SetMaterialAttributes!(
   MaterialAttributes
   EmissiveColor=Add(A=Multiply(A=@n1.RGB, B=@n20), B=Multiply(A=@n24, B=@n25))
   ShadingModel=ShadingModel()
   Opacity=@n27
   OpacityMask=@n27
  ) {attributeSetTypes [
    #1
    #2
    #3
    #4
   ]}
  n29 = Constant()
  n30 = ScalarParameter() {defaultValue=0.4, group=Bittenhead, parameterName=BH_DiffMod-Morning}
  n31 = ScalarParameter() {defaultValue=1, group=Bittenhead, parameterName=BH_DiffMod-Day}
  n32 = ScalarParameter() {defaultValue=0.4, group=Bittenhead, parameterName=BH_DiffMod-Eve}
  n33 = ScalarParameter() {defaultValue=0.4, group=Bittenhead, parameterName=BH_DiffMod-Night}
  n34 = GetMaterialAttributes(None=FunctionInput() {inputName=Material, inputType=FunctionInput_MaterialAttributes}) {attributeGetTypes [#5, #1, #6]}
  n35 = VectorParameter() {group=Bittenhead, parameterName=BH_DiffuseColorMod, defaultValue=<rgba 1 1 1 1>}
  n36 = Saturate(None=Multiply(
    A=Multiply(A=@n34.BaseColor, B=@n35.RGB)
    B=MaterialFunctionCall(Value_Morning=@n30, Value_Day=@n31, Value_Evening=@n32, Value_Night=@n33) {materialFunction=&Scalar_Time_of_Day}.Result
   ))
  n37 = CollectionParameter() {
   collection=&FortniteMaterialParameters
   parameterName=SunlightDirectionalVector
  }
  n38 = ComponentMask(None=Transform(None=@n34.Normal)) {b=true, g=true, r=true}
  n39 = LinearInterpolate(A=@n38, B=ComponentMask(None=VertexNormalWS()) {b=true, g=true, r=true}, Alpha=Constant() {r=0.25})
  n40 = ShadingPathSwitch(Default=@n39, Deferred=@n39, Forward=@n38, Mobile=@n38)
  n41 = Subtract(A=DotProduct(A=ComponentMask(None=@n37) {b=true, g=true, r=true}, B=@n40)) {constB=-1}
  n42 = Saturate(None=OneMinus(None=@n41))
  n43 = Constant() {r=2}
  n44 = Divide(A=@n41, B=ScalarParameter() {defaultValue=1.75, group=Bittenhead, parameterName=BH_LightLevels})
  n45 = Saturate(None=Add(A=Saturate(None=@n44), B=Power(Base=@n42, Exp=@n43)))
  n46 = ScalarParameter() {defaultValue=0.5, group=Bittenhead, parameterName=BH_HeightLerpAmt}
  n47 = Saturate(None=Divide(A=@n41, B=ScalarParameter() {defaultValue=10, group=Bittenhead, parameterName=BH_HHTE-Div}))
  n48 = ScalarParameter() {defaultValue=4.75, group=Bittenhead, parameterName=BH_HHTE-Pow}
  n49 = Saturate(None=Multiply(A=Power(Base=@n47, Exp=@n48), B=ScalarParameter() {defaultValue=1000, group=Bittenhead, parameterName=BH_HHTE-Mod}))
  n50 = Add(
   A=Multiply(
    A=@n36
    B=MaterialFunctionCall(Transition Phase=@n45, Height Texture=@n46) {materialFunction=&HeightLerp}.Lerp Alpha No Contrast
   )
   B=Multiply(A=@n49, B=ScalarParameter() {defaultValue=0.03, group=Bittenhead, parameterName=BH_HHTE-Amt})
  )
  n51 = Divide(A=@n41, B=ScalarParameter() {defaultValue=2, group=Bittenhead, parameterName=BH_SSL-Div})
  n52 = Multiply(
   A=Power(Base=Saturate(None=@n51), Exp=ScalarParameter() {defaultValue=50, group=Bittenhead, parameterName=BH_SSL-Pow})
   B=ScalarParameter() {defaultValue=10, group=Bittenhead, parameterName=BH_SSL-Mod}
  )
  n53 = ScalarParameter() {defaultValue=3, group=Bittenhead, parameterName=BH_SSL-Morning}
  n54 = ScalarParameter() {defaultValue=4, group=Bittenhead, parameterName=BH_SSL-Day}
  n55 = ScalarParameter() {defaultValue=3, group=Bittenhead, parameterName=BH_SSL-Eve}
  n56 = ScalarParameter() {defaultValue=2, group=Bittenhead, parameterName=BH_SSL-Night}
  n57 = Multiply(
   A=@n50
   B=Multiply(
    A=Saturate(None=@n52)
    B=MaterialFunctionCall(Value_Morning=@n53, Value_Day=@n54, Value_Evening=@n55, Value_Night=@n56) {materialFunction=&Scalar_Time_of_Day}.Result
   )
  )
  n58 = DotProduct(A=@n40, B=CameraVectorWS()) {desc=N Dot C - Fresnel from Normals}
  n59 = LinearInterpolate(
   A=ScalarParameter() {defaultValue=-3, group=Bittenhead, parameterName=BH_FresLerpA}
   B=ScalarParameter() {defaultValue=3, group=Bittenhead, parameterName=BH_FresLerpB}
   Alpha=OneMinus(None=Saturate(None=@n58))
  )
  n60 = Multiply(A=Multiply(A=Saturate(None=@n59), B=ScalarParameter() {defaultValue=30, group=Bittenhead, parameterName=BH_FresMod}), B=@n49)
  n61 = Saturate(None=Add(A=Add(A=@n50, B=@n57), B=Multiply(A=@n36, B=@n60)))
  n62 = SetMaterialAttributes!(
   MaterialAttributes=@n34.MaterialAttributes
   BaseColor=@n29
   EmissiveColor=Add(A=Multiply(A=@n61, B=@n13), B=@n19)
   Specular=@n29
   Roughness=Constant() {r=1}
   Metallic=@n29
   ShadingModel=ShadingModel()
  ) {attributeSetTypes [
    #5
    #1
    #7
    #8
    #9
    #2
   ]}
  FunctionOutput(None=StaticSwitchParameter(True=@n28, False=@n62) {parameterName=IsSunglasses})
  PixelNormalWS()
  StaticSwitchParameter(True=ObjectPositionWS(), False=@n6) {group=Noir, parameterName=BH_UseObjectPos}
  n63 = VectorParameter() {parameterName=zzzx, defaultValue=<rgba 0.5 0 0 1>}
  n64 = VectorParameter() {parameterName=zzzy, defaultValue=<rgba -0.5 0 0 1>}
  n65 = ScalarParameter() {defaultValue=0.005, group=Bittenhead, parameterName=cccc}
  n66 = Constant2Vector() {g=1}
  n67 = Multiply(A=Time(), B=ScalarParameter() {defaultValue=0.3, group=Bittenhead, parameterName=cccc2})
  n68 = TextureSample(UVs=Add(A=Multiply(A=@n65, B=@n8.UVs), B=Multiply(A=@n66, B=@n67))) {samplerSource=SSM_Wrap_WorldGroupSettings, texture=&T_Vert_Gradient_Sections}
  Transform(None=Multiply(A=LinearInterpolate(A=@n63.RGB, B=@n64.RGB, Alpha=@n68.R), B=@n68.R)) {transformSourceType=TRANSFORMSOURCE_View}
 ]}
