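# Skinned character material for the male medium BittenHead body, using material attributes with Default Lit shading.
# The shared character PreFX function builds the base, and a UseBanana static switch optionally routes it through a Banana function borrowed from weapon skins.
# The selected result passes through MF_Bittenhead, then the shared PostFX function.
# PreFX's AO output feeds the PostFX AO, Distance Blend and Fresnel inputs alike.
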Material /BRCosmetics/Characters/Player/Male/Medium/Bodies/M_MED_BittenHead/Materials/M_Bittenhead
graph {
 properties {
  bUseMaterialAttributes=true
  bUsedWithMeshDeformer=true
  bUsedWithMorphTargets=true
  bUsedWithSkeletalMesh=true
  pixelDepthOffsetMode=PDOM_Legacy
  translucencyLightingMode=TLM_SurfacePerPixelLighting
  usedShadingModels=Default Lit
 }
 nodes [
  n1 = MaterialFunctionCall() {materialFunction=&BaseFN_CharacterFunction_PreFX}
  n2 = MaterialFunctionCall!(Gmap, Material=@n1.Result) {materialFunction=/Game/Weapons/WeaponSkins/MaterialFunction/Banana}
  MaterialFunctionCall(
   AO=@n1.AO
   Distance Blend=@n1.AO
   Fresnel=@n1.AO
   Material=MaterialFunctionCall!(Material=StaticSwitchParameter(True=@n2.Result, False=@n1.Result) {parameterName=UseBanana}) {materialFunction=&/MF_Bittenhead}.Result
  ) {materialFunction=&BaseFN_CharacterFunction_PostFX}
 ]
}
