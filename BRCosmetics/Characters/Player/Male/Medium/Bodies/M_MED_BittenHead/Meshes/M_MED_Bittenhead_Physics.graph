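# Physics asset for the male BittenHead body cosmetic, providing bodies for secondary motion of the jacket and tie.
# Seven two-box jacket chains hang from master capsules on the pelvis, and a five-box tie chain hangs from spine_05, with swing and tight twist limits along each chain.
# Pelvis, spine and thigh capsules form the body volume.
# Angularly free constraints also link each front panel's tip to its neighboring side panel.
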PhysicsAsset /BRCosmetics/Characters/Player/Male/Medium/Bodies/M_MED_BittenHead/Meshes/M_MED_Bittenhead_Physics
bodies {
 dyn_jacket_bk_01 [{shapeType=Box, extentX=7.6875, extentY=14.25, extentZ=0.8125, center=<3.9737 0 0>}]
 dyn_jacket_bk_02 [{shapeType=Box, extentX=7.6875, extentY=14.25, extentZ=0.8125, center=<3.9737 0 0>}]
 dyn_jacket_bk_MSTR [{shapeType=Capsule, length=1, radius=0.5, rotation=<rot 90 0 0>}]
 dyn_jacket_bk_l_01 [{shapeType=Box, extentX=7.6875, extentY=14.25, extentZ=0.8125, center=<3.9737 0 0>}]
 dyn_jacket_bk_l_02 [{shapeType=Box, extentX=7.6875, extentY=14.25, extentZ=0.8125, center=<3.9737 0 0>}]
 dyn_jacket_bk_l_MSTR [{shapeType=Capsule, length=1, radius=0.5, rotation=<rot 90 0 0>}]
 dyn_jacket_bk_r_01 [{shapeType=Box, extentX=7.6875, extentY=14.25, extentZ=0.8125, center=<3.9737 0 0>}]
 dyn_jacket_bk_r_02 [{shapeType=Box, extentX=7.6875, extentY=14.25, extentZ=0.8125, center=<3.9737 0 0>}]
 dyn_jacket_bk_r_MSTR [{shapeType=Capsule, length=1, radius=0.5, rotation=<rot 90 0 0>}]
 dyn_jacket_fr_l_01 [{shapeType=Box, extentX=13.75, extentY=11.5, extentZ=1.625, center=<6.3542 0 -0.0001>}]
 dyn_jacket_fr_l_02 [{shapeType=Box, extentX=10.6875, extentY=11.5, extentZ=1.625, center=<5.1064 0 0>}]
 dyn_jacket_fr_l_MSTR [{shapeType=Capsule, length=1, radius=0.5, rotation=<rot 90 0 0>}]
 dyn_jacket_fr_r_01 [{shapeType=Box, extentX=16.4375, extentY=15.75, extentZ=1.75, center=<8.3025 0 -0.0001>}]
 dyn_jacket_fr_r_02 [{shapeType=Box, extentX=11.4864, extentY=15.75, extentZ=2.2354, center=<5.3044 0 -0.0116>}]
 dyn_jacket_fr_r_MSTR [{shapeType=Capsule, length=1, radius=0.5, rotation=<rot 90 0 0>}]
 dyn_jacket_l_01 [{shapeType=Box, extentX=7.6875, extentY=14.25, extentZ=0.8125, center=<3.9737 0 0>}]
 dyn_jacket_l_02 [{shapeType=Box, extentX=7.6875, extentY=14.25, extentZ=0.8125, center=<3.9737 0 0>}]
 dyn_jacket_l_MSTR [{shapeType=Capsule, length=1, radius=0.5, rotation=<rot 90 0 0>}]
 dyn_jacket_r_01 [{shapeType=Box, extentX=7.6875, extentY=14.25, extentZ=0.8125, center=<3.9737 0 0>}]
 dyn_jacket_r_02 [{shapeType=Box, extentX=7.6875, extentY=14.25, extentZ=0.8125, center=<3.9737 0 0>}]
 dyn_jacket_r_MSTR [{shapeType=Capsule, length=1, radius=0.5, rotation=<rot 90 0 0>}]
 dyn_tie_01 [{shapeType=Box, extentX=12.4375, extentY=4.5, extentZ=1.375, center=<5.8529 0 0.0001>}]
 dyn_tie_02 [{shapeType=Box, extentX=13.875, extentY=4.5, extentZ=2.25, center=<6.0533 0 0.0001>}]
 dyn_tie_03 [{shapeType=Box, extentX=11.5625, extentY=5, extentZ=1.375, center=<4.9506 0 0.0001>}]
 dyn_tie_04 [{shapeType=Box, extentX=8.9375, extentY=5.5625, extentZ=1.375, center=<4.0807 0 0>}]
 dyn_tie_05 [{shapeType=Box, extentX=4.75, extentY=5.875, extentZ=0.8125, center=<2.2241 0 0.0001>}]
 dyn_tie_MSTR [{shapeType=Capsule, length=1, radius=0.5, rotation=<rot 90 0 0>}]
 pelvis [{shapeType=Capsule, length=6.8933, radius=10.7942, center=<-2.0733 -1.0713 0>}]
 spine_01 [{shapeType=Capsule, length=19.1562, radius=12.0938, center=<0 -0.1776 0>}]
 spine_02 [{shapeType=Capsule, length=19.1562, radius=12.0938, center=<0 -0.6181 0>}]
 spine_03 [{shapeType=Capsule, length=19.1562, radius=12.0938, center=<0 -0.2866 0>}]
 spine_04 [{shapeType=Capsule, length=19.1562, radius=12.0938, center=<0 -0.1776 0>}]
 spine_05 [{shapeType=Capsule, length=20, radius=13.2188, center=<-4.5203 -1.893 0>}]
 thigh_l [{shapeType=Capsule, length=26.75, radius=8.625, center=<-13.9111 0.0005 0.9087>, rotation=<rot 90 0 0>}]
 thigh_r [{shapeType=Capsule, length=21.7778, radius=9.316, center=<12.8168 -0.3902 -1.2731>, rotation=<rot 90 0 0>}]
}
constraints [
 {bone1Name=dyn_tie_01, bone2Name=dyn_tie_MSTR, swing1LimitDegrees=60, swing2LimitDegrees=60, twistLimitDegrees=5}
 {bone1Name=dyn_tie_02, bone2Name=dyn_tie_01, swing1LimitDegrees=60, swing2LimitDegrees=60, twistLimitDegrees=5}
 {bone1Name=dyn_tie_03, bone2Name=dyn_tie_02, swing1LimitDegrees=60, swing2LimitDegrees=60, twistLimitDegrees=5}
 {bone1Name=dyn_tie_04, bone2Name=dyn_tie_03, swing1LimitDegrees=60, swing2LimitDegrees=60, twistLimitDegrees=5}
 {bone1Name=dyn_tie_05, bone2Name=dyn_tie_04, swing1LimitDegrees=60, swing2LimitDegrees=60, twistLimitDegrees=5}
 {bone1Name=dyn_jacket_l_01, bone2Name=dyn_jacket_l_MSTR, swing1LimitDegrees=60, swing2LimitDegrees=60, twistLimitDegrees=5}
 {bone1Name=dyn_jacket_l_02, bone2Name=dyn_jacket_l_01, swing1LimitDegrees=60, swing2LimitDegrees=60, twistLimitDegrees=5}
 {bone1Name=dyn_jacket_fr_l_01, bone2Name=dyn_jacket_fr_l_MSTR, swing1LimitDegrees=5, swing2LimitDegrees=80, twistLimitDegrees=5}
 {bone1Name=dyn_jacket_fr_l_02, bone2Name=dyn_jacket_fr_l_01, swing1LimitDegrees=5, swing2LimitDegrees=30, twistLimitDegrees=5}
 {bone1Name=dyn_jacket_bk_l_01, bone2Name=dyn_jacket_bk_l_MSTR, swing1LimitDegrees=60, swing2LimitDegrees=35, twistLimitDegrees=5}
 {bone1Name=dyn_jacket_bk_l_02, bone2Name=dyn_jacket_bk_l_01, swing1LimitDegrees=60, swing2LimitDegrees=35, twistLimitDegrees=5}
 {bone1Name=dyn_jacket_bk_01, bone2Name=dyn_jacket_bk_MSTR, swing1LimitDegrees=60, swing2LimitDegrees=35, twistLimitDegrees=5}
 {bone1Name=dyn_jacket_bk_02, bone2Name=dyn_jacket_bk_01, swing1LimitDegrees=60, swing2LimitDegrees=35, twistLimitDegrees=5}
 {bone1Name=dyn_jacket_fr_r_01, bone2Name=dyn_jacket_fr_r_MSTR, swing1LimitDegrees=60, swing2LimitDegrees=60, twistLimitDegrees=5}
 {bone1Name=dyn_jacket_fr_r_02, bone2Name=dyn_jacket_fr_r_01, swing1LimitDegrees=60, swing2LimitDegrees=60, twistLimitDegrees=5}
 {bone1Name=dyn_jacket_bk_r_01, bone2Name=dyn_jacket_bk_r_MSTR, swing1LimitDegrees=60, swing2LimitDegrees=35, twistLimitDegrees=5}
 {bone1Name=dyn_jacket_bk_r_02, bone2Name=dyn_jacket_bk_r_01, swing1LimitDegrees=60, swing2LimitDegrees=35, twistLimitDegrees=5}
 {bone1Name=dyn_jacket_r_01, bone2Name=dyn_jacket_r_MSTR, swing1LimitDegrees=60, swing2LimitDegrees=60, twistLimitDegrees=5}
 {bone1Name=dyn_jacket_r_02, bone2Name=dyn_jacket_r_01, swing1LimitDegrees=60, swing2LimitDegrees=60, twistLimitDegrees=5}
 {
  bone1Name=dyn_tie_MSTR
  bone2Name=spine_05
  swing1LimitDegrees=45
  swing1Motion=Free
  swing2LimitDegrees=45
  swing2Motion=Free
  twistLimitDegrees=45
  twistMotion=Free
 }
 {
  bone1Name=spine_05
  bone2Name=spine_04
  swing1LimitDegrees=45
  swing1Motion=Free
  swing2LimitDegrees=45
  swing2Motion=Free
  twistLimitDegrees=45
  twistMotion=Free
 }
 {
  bone1Name=spine_04
  bone2Name=spine_03
  swing1LimitDegrees=45
  swing1Motion=Free
  swing2LimitDegrees=45
  swing2Motion=Free
  twistLimitDegrees=45
  twistMotion=Free
 }
 {
  bone1Name=dyn_jacket_l_MSTR
  bone2Name=pelvis
  swing1LimitDegrees=45
  swing1Motion=Free
  swing2LimitDegrees=45
  swing2Motion=Free
  twistLimitDegrees=45
  twistMotion=Free
 }
 {
  bone1Name=dyn_jacket_fr_l_MSTR
  bone2Name=pelvis
  swing1LimitDegrees=45
  swing1Motion=Free
  swing2LimitDegrees=45
  swing2Motion=Free
  twistLimitDegrees=45
  twistMotion=Free
 }
 {
  bone1Name=dyn_jacket_bk_l_MSTR
  bone2Name=pelvis
  swing1LimitDegrees=45
  swing1Motion=Free
  swing2LimitDegrees=45
  swing2Motion=Free
  twistLimitDegrees=45
  twistMotion=Free
 }
 {
  bone1Name=dyn_jacket_bk_MSTR
  bone2Name=pelvis
  swing1LimitDegrees=45
  swing1Motion=Free
  swing2LimitDegrees=45
  swing2Motion=Free
  twistLimitDegrees=45
  twistMotion=Free
 }
 {
  bone1Name=dyn_jacket_fr_r_MSTR
  bone2Name=pelvis
  swing1LimitDegrees=45
  swing1Motion=Free
  swing2LimitDegrees=45
  swing2Motion=Free
  twistLimitDegrees=45
  twistMotion=Free
 }
 {
  bone1Name=dyn_jacket_bk_r_MSTR
  bone2Name=pelvis
  swing1LimitDegrees=45
  swing1Motion=Free
  swing2LimitDegrees=45
  swing2Motion=Free
  twistLimitDegrees=45
  twistMotion=Free
 }
 {
  bone1Name=dyn_jacket_r_MSTR
  bone2Name=pelvis
  swing1LimitDegrees=45
  swing1Motion=Free
  swing2LimitDegrees=45
  swing2Motion=Free
  twistLimitDegrees=45
  twistMotion=Free
 }
 {
  bone1Name=thigh_l
  bone2Name=pelvis
  swing1LimitDegrees=45
  swing1Motion=Free
  swing2LimitDegrees=45
  swing2Motion=Free
  twistLimitDegrees=45
  twistMotion=Free
 }
 {
  bone1Name=thigh_r
  bone2Name=pelvis
  swing1LimitDegrees=45
  swing1Motion=Free
  swing2LimitDegrees=45
  swing2Motion=Free
  twistLimitDegrees=45
  twistMotion=Free
 }
 {
  bone1Name=spine_03
  bone2Name=spine_02
  swing1LimitDegrees=45
  swing1Motion=Free
  swing2LimitDegrees=45
  swing2Motion=Free
  twistLimitDegrees=45
  twistMotion=Free
 }
 {
  bone1Name=spine_02
  bone2Name=spine_01
  swing1LimitDegrees=45
  swing1Motion=Free
  swing2LimitDegrees=45
  swing2Motion=Free
  twistLimitDegrees=45
  twistMotion=Free
 }
 {
  bone1Name=spine_01
  bone2Name=pelvis
  swing1LimitDegrees=45
  swing1Motion=Free
  swing2LimitDegrees=45
  swing2Motion=Free
  twistLimitDegrees=45
  twistMotion=Free
 }
 {
  bone1Name=dyn_jacket_fr_r_02
  bone2Name=dyn_jacket_r_02
  swing1LimitDegrees=45
  swing1Motion=Free
  swing2LimitDegrees=45
  swing2Motion=Free
  twistLimitDegrees=45
  twistMotion=Free
 }
 {
  bone1Name=dyn_jacket_l_02
  bone2Name=dyn_jacket_fr_l_02
  swing1LimitDegrees=45
  swing1Motion=Free
  swing2LimitDegrees=45
  swing2Motion=Free
  twistLimitDegrees=45
  twistMotion=Free
 }
]
